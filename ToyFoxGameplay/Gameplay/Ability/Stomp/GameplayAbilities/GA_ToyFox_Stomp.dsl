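; Gameplay ability for the ToyFox giant pawn's stomp, which damages pawns and destroys or damages the world.
; Activation plays a montage, applies an active effect and alerts AI with noise; the hit notify triggers targeting.
; The server applies damage asynchronously or queues buildings in the destruction manager, separately damaging destruction-immune actors, while clients predict building destruction by bounds size.
; CanActivateAbility runs a client-only sphere trace to require ground under the stomp.
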(event Ability|EventActivateAbilityFromEvent
  (params (EventData :type Gameplay%20Event%20Data%20Structure%20%28by%20ref%29))
  (bind _returnvalue (Fortnite|Athena|GameState|GetGameStateAthena))
  (bind _returnvalue_2 (Fortnite|Athena|GameState|GetGameStateAthena))
  ;[0.04186 0.15 0.009951 0.5] ActivateAbility: Play the stomp montage.
  (|CallParentFunction EventData)
  (bind (_asynctaskproxy _notifytag) (Ability|Tasks|PlayMontageAndWaitNotify "PlayMontageStomp" (Variables|Settings|GetStompMontage))
    (:then
      (Variables|Transient|SetAsyncTask_PlayMontage _asynctaskproxy)
      (bind _returnvalue_3 (Ability|ApplyGameplayEffectToOwner (Variables|Settings|GetStompActiveGameplayEffectClass)))
      (Variables|Transient|SetStompActiveGameplayEffectHandle _returnvalue_3)
      (Ability|CommitAbility)
      (Ability|ExecuteGameplayCueOnOwner (Utilities|Struct|BreakGameplayCueTag (Variables|GameplayCue|GetStompIntroGameplayCueTag)))
      ;;[0.03689 0.04106 0.15 0.5] Add noise event for AI to react to
      (bind _outfortplayerpawn (CallFunction|GetAvatarasValidFortPlayerPawn)
        (:ValidFortPlayerPawn
          (AI|Perception|MakeNoiseEventatLocation _outfortplayerpawn (ScalableFloat|GetValueatLevel (Variables|Hotfix|GetHotfix_PerceptionRange)) (Transformation|GetActorLocation _outfortplayerpawn) "WeaponFiring"))))
    (:OnCompleted
      _)
    (:OnBlendOut
      _)
    (:OnInterrupted
      _)
    (:OnCancelled
      _)
    (:OnNotifyBegin
      (if (== _notifytag "(TagName=\"ToyFox.AnimNotify.Ability.Stomp.Hit\")")
        (Ability|ApplyGameplayEffectToOwner (Variables|Settings|GetStompExecuteGameplayEffectClass))
        (Ability|HasAuthority
          (:True
            (Utilities|FlowControl|Sequence
              (:then_0
                (bind (_asynctaskproxy_2 _targetdata) (Ability|Tasks|PerformTargetingAbilityTask "StompTargetingPawn" (Variables|Settings|GetStompTargetingPresent) true false true)
                  (:then
                    (Variables|Transient|SetAsyncTask_TargetingPawn _asynctaskproxy_2))
                  (:Targeted
                    (Ability|Tasks|ApplyGameplayEffectToTargetDataAsync _targetdata (Variables|Settings|GetStompDamagePawnGameplayEffectClass) 1 1 (Variables|Hotfix|GetHotfix_MaxStompTargetsToProcessPerTick)))))
              (:then_1
                (bind (_asynctaskproxy_3 _targetdata_2) (Ability|Tasks|PerformTargetingAbilityTask "StompTargetingWorld" "/ToyFoxGameplay/Gameplay/Ability/Stomp/Targeting/TP_ToyFox_StompWorld.TP_ToyFox_StompWorld" true)
                  (:then
                    (Variables|Transient|SetAsyncTask_TargetingWorld _asynctaskproxy_3))
                  (:Targeted
                    ;;[0.15 0 0.1195 0.5] Instant destroy building actors.
                    (if (> (ScalableFloat|GetCachedScalableFloatValue (Variables|Hotfix|GetHotfix_BuildingActorsInstantDie)) 0)
                      (Utilities|IsValid _returnvalue
                        (:Is%20Valid
                          (bind _returnvalue_4 (Actor|GetComponentbyClass _returnvalue "/GiantPawnCore/Gameplay/Destruction/B_GiantPawnDestructionManagerGameStateComponent.B_GiantPawnDestructionManagerGameStateComponent_C"))
                          (Utilities|IsValid _returnvalue_4
                            (:Is%20Valid
                              (Public|AddBuildingsforDestruction _returnvalue_4 (Ability|TargetData|GetAllActorsfromTargetData _targetdata_2) "StompDestruction" (CallFunction|GetAvatarasFortPlayerPawn))
                              (Utilities|FlowControl|DelayUntilNextTick)
                              (bind _hitresults (CallFunction|GetGiantDestructionImmuneActors (Ability|TargetData|GetAllActorsfromTargetData _targetdata_2))) ; This will not include actors that have already been added for destruction. \nThe return array should be minimal in most cases.
                              (Ability|Tasks|ApplyGameplayEffectToHitResultsAsync :OriginatingAbilitySystemComponent (Ability|GetAbilitySystemComponentfromActorInfo) :HitResults _hitresults :GameplayEffectClass (Variables|Settings|GetStompDamageWorldGameplayEffectClass) :GameplayEffectLevel 0 :Stacks 1 :MaxTargetsToProcessPerTickCVar (Variables|Hotfix|GetHotfix_MaxStompTargetsToProcessPerTick))))))
                      (else
                        ;; Async apply damage to buildings
                        (Ability|Tasks|ApplyGameplayEffectToTargetDataAsync _targetdata_2 (Variables|Settings|GetStompDamageWorldGameplayEffectClass) 1 1 (Variables|Hotfix|GetHotfix_MaxStompTargetsToProcessPerTick)))))))))
          (:False
            ;;[0 0.15 0.1232 0.5] On the client predict an instant reaction to the stomp action.
            (bind (_asynctaskproxy_4 _targetdata_3) (Ability|Tasks|PerformTargetingAbilityTask "StompTargeting_ClientPrediction" (Variables|Settings|GetStompClientPredictionTargetingPresent) false true true)
              (:then
                (Variables|Transient|SetAsyncTask_TargetingPawn _asynctaskproxy_4))
              (:Targeted
                (bind _outfortplayerpawn_2 (CallFunction|GetAvatarasValidFortPlayerPawn)
                  (:ValidFortPlayerPawn
                    (bind _asb_toy_fox_giant_player_pawn (Utilities|Casting|CastToB_ToyFoxGiantPlayerPawn _outfortplayerpawn_2))
                    (bind _toyfoxgiantpawndestructioncomponent (Variables|/ToyFoxGameplay/Gameplay/B_ToyFoxGiantPlayerPawn.B_ToyFoxGiantPlayerPawn_C|GetToyFoxGiantPawnDestructionComponent _asb_toy_fox_giant_player_pawn)
                      (:then
                        (Utilities|IsValid _returnvalue_2
                          (:Is%20Valid
                            (bind _returnvalue_5 (Actor|GetComponentbyClass _returnvalue_2 "/GiantPawnCore/Gameplay/Destruction/B_GiantPawnDestructionManagerGameStateComponent.B_GiantPawnDestructionManagerGameStateComponent_C"))
                            (Utilities|IsValid _returnvalue_5
                              (:Is%20Valid
                                (bind _output_get (Variables|Transient|SetTempTargetedActors (Ability|TargetData|GetAllActorsfromTargetData _targetdata_3)))
                                (for _array_element _output_get
                                  (bind (_candestroy _asbuildingactor) (Helper|IsActorValidforDestruction _returnvalue_5 _array_element))
                                  (if _candestroy
                                    (bind _asbuilding_static_mesh (Utilities|Casting|CastToBuildingSMActor _asbuildingactor))
                                    (if (> (Math|Vector|VectorLengthSquared (.BoxExtent (Collision|GetActorBounds _asbuilding_static_mesh true false))) (ScalableFloat|GetCachedScalableFloatValue (Variables|Hotfix|GetHotfix_BGASqrBoundsToPredictDeath)))
                                      (PredictiveDestruction|PredictDeathforBuildingActorAsync _toyfoxgiantpawndestructioncomponent _asbuilding_static_mesh (CallFunction|GetAvatarasFortPlayerPawn) false)
                                      (else
                                        (PredictiveDestruction|SimplePredictDeathforBuildingActorDeferred _toyfoxgiantpawndestructioncomponent _asbuilding_static_mesh (ScalableFloat|GetCachedScalableFloatValue (Variables|Hotfix|GetHotfix_PredictDestroyDelay)))))))
                                (Utilities|Array|Clear (Variables|Transient|GetTempTargetedActors)))))))
                      (:else)))))))))))
  (Ability|EndAbility))

(event Ability|EventOnEndAbility
  (params (bWasCancelled :type Boolean))
  ;[0.04186 0.15 0.009951 0.5] ActivateAbility: Play the stomp montage.
  ;[0.05208 0.03226 0 0.5] EndAbility: Clean up tasks.
  (Utilities|FlowControl|Sequence
    (:then_0
      (GameplayTasks|EndTask (Variables|Transient|GetAsyncTask_TargetingPawn))
      (Ability|RemoveGameplayEffectFromOwnerWithHandle (Variables|Transient|GetStompActiveGameplayEffectHandle)))
    (:then_1
      (|CallParentFunction bWasCancelled))))

(fn GetGiantDestructionImmuneActors
  (params (InActors :type Array%20of%20Actor%20Object%20References))
  (Utilities|FlowControl|Sequence
    (:then_0
      (for _array_element InActors
        (Utilities|IsValid _array_element
          (:Is%20Valid
            (if (Actor|ActorHasTag _array_element "GiantDestructionImmune")
              (Utilities|Array|Add (LocalVariables|GetReturnResults) (Collision|MakeHitResult true false 0.0 0.0 "0, 0, 0" "0, 0, 0" "0,0,1" "0,0,1" 0 _array_element)))))))
    (:then_1
      (return (LocalVariables|GetReturnResults)))))

(fn CanActivateAbility
  (params (ActorInfo :type Gameplay%20Ability%20Actor%20Info%20Structure) (Handle :type Gameplay%20Ability%20Spec%20Handle%20Structure))
  (bind (_owneractor _avataractor) (Utilities|Struct|BreakGameplayAbilityActorInfo ActorInfo))
  (bind _returnvalue (Fortnite|GetFortPlayerControllerfromActor _avataractor))
  ;[0 0.15 0.09664 0.5] CanActivateAbility: On the client, check to see if the character is facing the intended direction and also has a valid ground to stomp on.
  (Utilities|FlowControl|Sequence
    (:then_0
      (bind _asfort_player_pawn (Utilities|Casting|CastToFortPlayerPawn _avataractor))
      (bind _mesh (Variables|/Script/Engine.Character|GetMesh _asfort_player_pawn)
        (:then
          (bind _returnvalue_2 (Transformation|GetSocketTransform _mesh "pelvis" "RTS_World"))
          (Utilities|IsValid _returnvalue
            (:Is%20Valid
              (if (not (Networking|HasAuthority _returnvalue))
                (if (> (ScalableFloat|GetCachedScalableFloatValue (Variables|Hotfix|GetHotfix_ValidateTrace)) 0)
                  (bind _output_get (LocalVariables|SetTraceStartLocation (Utilities|Operators|Add (.location _returnvalue_2) (* (Math|Vector|GetRightVector (.rotation _returnvalue_2)) (ScalableFloat|GetCachedScalableFloatValue (Variables|Hotfix|GetHotfix_ValidateTraceLocOffsetForward))) (* (Math|Vector|GetForwardVector (.rotation _returnvalue_2)) (ScalableFloat|GetCachedScalableFloatValue (Variables|Hotfix|GetHotfix_ValidateTraceLocOffsetLeft))) (Math|Vector|MakeVector 0.0 0.0 (ScalableFloat|GetCachedScalableFloatValue (Variables|Hotfix|GetHotfix_ValidateTraceStartOffset))))))
                  (bind (_outhit _returnvalue_3) (Collision|SphereTraceByChannel _output_get (Math|Vector|MakeVector (.x _output_get) (.y _output_get) (- (- (.z (Transformation|GetActorLocation _asfort_player_pawn)) 2.5e+03) (ScalableFloat|GetCachedScalableFloatValue (Variables|Hotfix|GetHotfix_ValidateTraceEndOffset)))) (ScalableFloat|GetCachedScalableFloatValue (Variables|Hotfix|GetHotfix_ValidateTraceSphereRadius)) "TraceTypeQuery10" false (Utilities|Array|MakeArray _avataractor))) ; Only trace on the client for better responsiveness.\nValidating this is not the end of the world if not done on the server.\nPrimarily exists for cosmetic reasons.
                  (if _returnvalue_3
                    (return 0 _returnvalue_3)
                    (else
                      (return "(GameplayTags=((TagName=\"Activate.Failed.ToyFox.Stomp.NoGround\")))" false)))
                  (else
                    _)))
              (return 0 true))))
        (:else)))
    (:then_1
      (return 0 false))))
